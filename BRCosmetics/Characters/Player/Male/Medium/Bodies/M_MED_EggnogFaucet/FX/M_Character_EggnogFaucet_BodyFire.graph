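# Translucent, unlit, two-sided Niagara material for animated fire on the EggnogFaucet character, usable on sprites, ribbons and meshes.
# Emissive multiplies two panning color layers by particle color, with a shared noise reroute distorting UVs.
# Opacity combines flame texture, dissolve noise, shape masks, height-lerp disappearance, and view-angle and depth fades.
# Scalar-lerp toggles enable optional masks and vertex offsets: spline thickening, particle-driven bending and sine-noise normal displacement.
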Material /BRCosmetics/Characters/Player/Male/Medium/Bodies/M_MED_EggnogFaucet/FX/M_Character_EggnogFaucet_BodyFire
graph {
 properties {
  bAutomaticallySetUsageInEditor=false
  bUsedWithNiagaraMeshParticles=true
  bUsedWithNiagaraRibbons=true
  bUsedWithNiagaraSprites=true
  blendMode=BLEND_Translucent
  shadingModel=MSM_Unlit
  twoSided=true
  usedShadingModels=Unlit
 }
 nodes [
  n1 = Constant()
  n2 = DynamicParameter() {parameterIndex=1, paramNames [X,Y,Z,WPOIntensity]}
  n3 = MaterialFunctionCall(X=@n2.X, Y=@n2.X, Z=@n2.X) {materialFunction=&MakeFloat3}
  n4 = ParticlePositionWS()
  n5 = ScalarParameter() {defaultValue=35, parameterName=InterpolateWPORange}
  n6 = ComponentMask(None=TextureCoordinate()) {r=true}
  n7 = ScalarParameter() {defaultValue=3, parameterName=WPOBendPower}
  n8 = LinearInterpolate(
   B=Clamp(None=Subtract(A=@n3.Result, B=@n4), Min=Multiply(A=@n5) {constB=-1}, Max=@n5)
   Alpha=Multiply(A=@n2.WPOIntensity, B=Power(Base=@n6, Exp=@n7))
  )
  n9 = ScalarParameter() {parameterName=UseInterpolateWPO}
  n10 = Constant()
  n11 = DynamicParameter() {paramNames [PannerSpeedX_Offset, Spline_Base, Spline_Tip, WPO_Offset]}
  n12 = MaterialFunctionCall(
   WidthBase=Multiply(A=ScalarParameter() {defaultValue=5, parameterName=Spline_Base}, B=@n11.Spline_Base)
   WidthTip=Multiply(A=ScalarParameter() {defaultValue=5, parameterName=Spline_Tip}, B=@n11.Spline_Tip)
   Expand U or V UV Channel=StaticBool()
  ) {materialFunction=&SplineThicken}
  n13 = ScalarParameter() {parameterName=Use_SplineThicken}
  n14 = TextureCoordinate()
  n15 = AppendVector(A=ScalarParameter() {defaultValue=1.5, parameterName=WPO_UTIling}, B=Constant() {r=1})
  n16 = DynamicParameter() {paramNames [PannerSpeedX_Offset, Spline_Base, Spline_Tip, WPO_Offset]}
  n17 = ComponentMask(None=Add(A=Multiply(A=@n14, B=@n15), B=@n16.WPO_Offset)) {r=true}
  n18 = Multiply(A=@n16.WPO_Offset, B=VertexColor().R)
  n19 = ScalarParameter() {parameterName=Use_VertexColor}
  n20 = Multiply(
   A=Multiply(A=Time(), B=ScalarParameter() {defaultValue=1, parameterName=WPO_Speed}) {constB=0.2}
   B=Multiply(A=Constant() {r=-1}, B=LinearInterpolate(A=@n16.WPO_Offset, B=@n18, Alpha=@n19))
  ) {constB=-1.5}
  n21 = TextureCoordinate()
  n22 = Add(
   A=Sine(None=Add(A=@n17, B=@n20))
   B=TextureSample(UVs=Panner(Coordinate=@n21) {speedX=-0.2}) {samplerType=SAMPLERTYPE_Normal, texture=&T_Weap_Distort_N_02}.R
  )
  n23 = ComponentMask(None=Add(A=TextureCoordinate() {uTiling=2}, B=Constant2Vector())) {r=true}
  n24 = Multiply(A=VertexNormalWS(), B=Saturate(None=Multiply(A=@n22, B=@n23)))
  n25 = ScalarParameter() {defaultValue=-1.5, parameterName=WPOMult}
  Add(
   A=Add(A=LinearInterpolate(A=@n1, B=@n8, Alpha=@n9), B=LinearInterpolate(A=@n10, B=@n12.WorldPositionOffset, Alpha=@n13))
   B=LinearInterpolate(A=Constant(), B=Multiply(A=@n24, B=@n25), Alpha=ScalarParameter() {parameterName=Use_WPO})
  )
  n26 = ParticleColor()
  n27 = TextureCoordinate()
  n28 = ScalarParameter() {defaultValue=0.5, parameterName=AlphaTex_UTiling}
  n29 = VertexColor()
  n30 = AppendVector(A=LinearInterpolate(A=@n28, B=Multiply(A=@n28, B=@n29.R), Alpha=ScalarParameter() {parameterName=Use_VertexColor}), B=Constant() {r=1})
  n31 = Multiply(A=@n27, B=AppendVector(A=Constant() {r=0.5}, B=Constant() {r=1}))
  n32 = AppendVector(A=Add(A=Constant() {r=1}, B=ParticleRandom()), B=Constant() {r=1})
  n33 = Add(
   A=LinearInterpolate(A=Multiply(A=@n27, B=@n30), B=Add(A=@n31, B=@n32), Alpha=ScalarParameter() {parameterName=Use_ParticleRandom})
   B=Multiply(A=NamedRerouteUsage(), B=ScalarParameter() {defaultValue=0.2, parameterName=NoiseIntensity}) {constB=0.1}
  )
  n34 = DynamicParameter() {paramNames [PannerSpeedX_Offset, Spline_Base, Spline_Tip, WPO_Offset]}
  n35 = Multiply(A=@n34.PannerSpeedX_Offset, B=VertexColor().R)
  n36 = ScalarParameter() {parameterName=Use_VertexColor}
  n37 = AppendVector(
   A=Multiply(A=ScalarParameter() {defaultValue=-0.2, parameterName=SpeedPanner}, B=LinearInterpolate(A=@n34.PannerSpeedX_Offset, B=@n35, Alpha=@n36))
   B=Constant()
  )
  n38 = Multiply(
   A=@n26.A
   B=TextureSampleParameter2D(UVs=Panner(Coordinate=@n33, Speed=@n37) {speedX=-0.4}) {parameterName=AlphaTex, texture=/Game/Accessories/FORT_Backpacks/Backpack_M_MED_KeyChain/FX/T_Backpack_KeyChain_Flame}.R
  )
  n39 = Add(A=TextureCoordinate() {uTiling=0.8}, B=Constant2Vector() {g=0.3})
  n40 = Multiply(A=@n39, B=AppendVector(A=ScalarParameter() {defaultValue=1, parameterName=Disappear_UTiling}, B=Constant() {r=1}))
  n41 = Add(A=@n39, B=ParticleRandom())
  n42 = ScalarParameter() {parameterName=Use_ParticleRandom}
  n43 = ScalarParameter() {defaultValue=-0.5, parameterName=Disappear_SpeedX}
  n44 = Constant()
  n45 = TextureSampleParameter2D(UVs=Panner(Coordinate=LinearInterpolate(A=@n40, B=@n41, Alpha=@n42), Speed=AppendVector(A=@n43, B=@n44)) {speedX=-0.5}) {
   parameterName=DisappearTex
   samplerType=SAMPLERTYPE_LinearGrayscale
   texture=/Game/Global/Textures/Default/Caustics/T_Nosie_Caustics_Grayscale
  }
  n46 = ComponentMask(None=TextureCoordinate()) {r=true}
  n47 = Saturate(None=LinearInterpolate(
    A=ScalarParameter() {parameterName=Dissolve mask min}
    B=ScalarParameter() {defaultValue=0.5, parameterName=Dissolve mask max}
    Alpha=Power(Base=@n46) {constExponent=1.3}
   ))
  n48 = ComponentMask(None=TextureCoordinate()) {g=true}
  n49 = LinearInterpolate(Alpha=OneMinus(None=Sine(None=@n48) {period=2})) {constA=1, constB=3}
  n50 = LinearInterpolate(
   A=ScalarParameter() {defaultValue=-1.5, parameterName=Disappear min}
   B=ScalarParameter() {defaultValue=3, parameterName=Disappear max}
   Alpha=Power(Base=Power(Base=@n45.R) {constExponent=1.2}, Exp=Multiply(A=@n47, B=@n49))
  )
  n51 = Constant() {r=0.8}
  n52 = Constant() {r=1}
  n53 = TextureSample(UVs=Panner(Coordinate=TextureCoordinate()) {speedX=-0.4}) {samplerType=SAMPLERTYPE_Grayscale, texture=&LowResBlurredNoise}
  n54 = Multiply(
   A=Multiply(A=Saturate(None=@n50), B=LinearInterpolate(A=@n51, B=@n52, Alpha=@n53.R) {constA=0.3})
   B=SphereMask(A=TextureCoordinate(), B=Constant2Vector() {g=0.5, r=0.2}) {attenuationRadius=0.7, hardnessPercent=0}
  )
  n55 = Add(A=TextureCoordinate() {uTiling=0.8, vTiling=0.7}, B=Multiply(A=NamedRerouteUsage(), B=Constant() {r=0.2}) {constB=0.1})
  n56 = Constant2Vector() {g=0.35, r=0.6}
  n57 = ScalarParameter() {parameterName=LeftSphereMask_Hardness}
  n58 = Add(A=TextureCoordinate(), B=Multiply(A=NamedRerouteUsage(), B=Constant() {r=0.2}) {constB=0.1})
  n59 = Constant2Vector() {g=0.5, r=0.5}
  n60 = LinearInterpolate(
   A=Multiply(A=Multiply(A=@n38, B=@n54), B=SphereMask(A=@n55, B=@n56, Hardness=@n57) {attenuationRadius=0.5, hardnessPercent=70})
   B=Multiply(A=@n26.A, B=SphereMask(A=@n58, B=@n59) {attenuationRadius=0.3, hardnessPercent=50})
   Alpha=ScalarParameter() {parameterName=Use_FinalSphereMask}
  )
  n61 = DynamicParameter() {parameterIndex=2, paramNames [Disapper, SineWaveOffset, SineWavePWR, Param4]}
  n62 = ComponentMask(None=Add(A=TextureCoordinate(), B=@n61.SineWaveOffset)) {r=true}
  n63 = Saturate(None=Power(Base=Sine(None=@n62) {period=2}, Exp=@n61.SineWavePWR))
  n64 = TextureCoordinate() {uTiling=1.5}
  n65 = Multiply(
   A=LinearInterpolate(A=@n60, B=Multiply(A=@n60, B=@n63), Alpha=ScalarParameter() {parameterName=LerpOpacitySineWave})
   B=MaterialFunctionCall(
    Transition Phase=@n61.Disapper
    Height Texture=TextureSample(UVs=@n64) {samplerType=SAMPLERTYPE_Grayscale, texture=/Game/Global/Textures/Default/Noise/LowResBlurredNoiseNormalized}.R
   ) {materialFunction=&HeightLerp}.Alpha
  )
  n66 = ComponentMask(None=Add(A=TextureCoordinate(), B=Constant2Vector())) {r=true}
  n67 = Power(Base=Multiply(A=OneMinus(None=@n66)) {constB=5}) {constExponent=5}
  n68 = LinearInterpolate(A=@n65, B=Multiply(A=@n65, B=Saturate(None=@n67)), Alpha=ScalarParameter() {parameterName=Use_FinalLinearmask})
  n69 = ComponentMask(None=TextureCoordinate()) {r=true}
  n70 = ScalarParameter() {parameterName=LeftFade}
  n71 = OneMinus(None=@n69)
  n72 = ScalarParameter() {parameterName=RightFade}
  n73 = Multiply(A=Multiply(A=Power(Base=@n69, Exp=@n70) {constExponent=5}, B=Power(Base=@n71, Exp=@n72) {constExponent=1}), B=@n68)
  n74 = ScalarParameter() {parameterName=Use_Fade}
  n75 = ScalarParameter() {parameterName=MeshFade}
  DepthFade(
   Opacity=Multiply(
    A=LinearInterpolate(A=@n68, B=@n73, Alpha=@n74)
    B=MaterialFunctionCall("FadeSensitivity (optional)"=@n75) {materialFunction=&FadeBasedOnViewAngle}.Result
   )
   FadeDistance=ScalarParameter() {parameterName=FadeDistance}
  )
  n76 = ScalarParameter() {defaultValue=-0.1, group=TexB, parameterName=TexBmin}
  n77 = ScalarParameter() {defaultValue=1, group=TexB, parameterName=TexBmax}
  n78 = VectorParameter() {group=TexB, parameterName=TexBUVScale, defaultValue=<rgba 1 0.5 0 0>}
  n79 = Multiply(A=@n78.RGB, B=VertexColor().R)
  n80 = ScalarParameter() {parameterName=Use_VertexColor}
  n81 = Multiply(A=Add(A=TextureCoordinate(), B=Constant2Vector()), B=ComponentMask(None=LinearInterpolate(A=@n78.RGB, B=@n79, Alpha=@n80)) {g=true, r=true})
  n82 = Multiply(A=NamedRerouteUsage()) {constB=0.05}
  n83 = ScalarParameter() {defaultValue=-0.5, parameterName=TexBSpeedX}
  n84 = LinearInterpolate(A=@n83, B=Multiply(A=@n83, B=VertexColor().R), Alpha=ScalarParameter() {parameterName=Use_VertexColor})
  n85 = Constant()
  n86 = TextureSampleParameter2D(UVs=Panner(Coordinate=Add(A=@n81, B=@n82), Speed=AppendVector(A=@n84, B=@n85)) {speedX=-0.5}) {
   group=TexB
   parameterName=TexB
   samplerType=SAMPLERTYPE_LinearGrayscale
   texture=/Game/Effects/Fort_Effects/Emotes/UltraLight/Texture/T_Plasma_C_Greyscale
  }
  n87 = Multiply(
   A=VectorParameter() {group=TexB, parameterName=Color03, defaultValue=<rgba 2 1.3411 0.1771 1>}.RGB
   B=Power(
    Base=LinearInterpolate(A=@n76, B=@n77, Alpha=@n86.R) {constA=0.05, constB=1.2}
    Exp=ScalarParameter() {defaultValue=1, group=TexB, parameterName=TexBPower}
   ) {constExponent=1.5}
  )
  n88 = TextureCoordinate()
  n89 = LinearInterpolate(A=@n88, B=Add(A=@n88, B=ParticleRandom()), Alpha=ScalarParameter() {parameterName=Use_ParticleRandom})
  n90 = ComponentMask(None=VectorParameter() {group=TexA, parameterName=TexAUVScale, defaultValue=<rgba 2 1.2 0 0>}.RGB) {g=true, r=true}
  n91 = DynamicParameter() {paramNames [PannerSpeedX_Offset, Spline_Base, Spline_Tip, WPO_Offset]}
  n92 = VertexColor()
  n93 = Multiply(
   A=ScalarParameter() {defaultValue=-1.2, parameterName=Speed_Tex_A}
   B=LinearInterpolate(A=@n91.PannerSpeedX_Offset, B=Multiply(A=@n91.PannerSpeedX_Offset, B=@n92.R), Alpha=ScalarParameter() {parameterName=Use_VertexColor})
  )
  n94 = Constant()
  n95 = TextureSampleParameter2D(UVs=Panner(Coordinate=Multiply(A=@n89, B=@n90), Speed=AppendVector(A=@n93, B=@n94)) {speedX=-0.5}) {group=TexA, parameterName=TexA, samplerType=SAMPLERTYPE_Grayscale, texture=&LowResBlurredNoiseNormalized}
  n96 = LinearInterpolate(
   A=ScalarParameter() {defaultValue=0.2, group=TexA, parameterName=TexAmin}
   B=ScalarParameter() {defaultValue=1, group=TexA, parameterName=TexAmax}
   Alpha=@n95.R
  )
  n97 = NamedRerouteUsage()
  n98 = ComponentMask(None=Add(A=TextureCoordinate(), B=Multiply(A=@n97) {constB=0.1})) {g=true}
  n99 = OneMinus(None=Abs(None=Subtract(A=@n98) {constB=0.5}))
  n100 = Subtract(A=@n99) {constB=0.5}
  n101 = Power(Base=@n99) {constExponent=8}
  n102 = Multiply(A=Add(A=Multiply(A=@n96, B=@n100), B=Multiply(A=@n101) {constB=0.5})) {constB=2}
  n103 = LinearInterpolate(
   A=VectorParameter() {group=TexA, parameterName=Color1, defaultValue=<rgba 1 0.0621 0.0033 1>}.RGB
   B=VectorParameter() {group=TexA, parameterName=Color2, defaultValue=<rgba 5 1.1421 0.1364 1>}.RGB
   Alpha=Saturate(None=Power(Base=@n102) {constExponent=2.5})
  )
  MaterialFunctionCall(Emissive=Multiply(A=Multiply(A=@n87, B=@n103), B=@n26.RGB)) {materialFunction=&MF_FX_EmissiveAdjustments}
  n104 = TextureCoordinate() {uTiling=0.7, vTiling=0.7}
  n105 = Add(A=@n104, B=Constant2Vector() {g=0.05})
  n106 = Add(A=@n104, B=ParticleRandom())
  n107 = ScalarParameter() {parameterName=Use_ParticleRandom}
  n108 = ScalarParameter() {defaultValue=-0.1, parameterName=NoiseTex_SpeedX}
  n109 = LinearInterpolate(A=@n108, B=Multiply(A=@n108, B=ParticleRandom()), Alpha=@n107)
  n110 = Constant()
  n111 = TextureSample(UVs=Panner(Coordinate=LinearInterpolate(A=@n105, B=@n106, Alpha=@n107), Speed=AppendVector(A=@n109, B=@n110)) {speedX=-0.1}) {samplerType=SAMPLERTYPE_Masks, texture=/Game/Weapons/WeaponSkins/Wraps/Textures/Mash/Mash_Noise}
  NamedRerouteDeclaration(None=ComponentMask(None=LinearInterpolate(Alpha=@n111.RGB) {constA=0.1}) {g=true, r=true}) {name=Noise, nodeColor=<rgba 0.0078 1 0 1>}
 ]
}
