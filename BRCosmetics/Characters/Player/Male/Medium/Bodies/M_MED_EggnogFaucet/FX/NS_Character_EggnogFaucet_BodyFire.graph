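# Niagara body fire cosmetic for the EggnogFaucet player character, layering fire rings, ray meshes, glow sprites and lights.
# Partition-indexed arrays give each particle its bone, offset and scale on the skeletal mesh.
# Duplicated 001 emitters split the effect through zero-scale masks, letting hand-hidden and foot-hidden flags hide separate parts.
# Front-end preview adjusts placement and opacity, the body ring fades by distance, and owner speed gates running sparks.
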NiagaraSystem /BRCosmetics/Characters/Player/Male/Medium/Bodies/M_MED_EggnogFaucet/FX/NS_Character_EggnogFaucet_BodyFire
dependencies {
 usedDataInterfaces [
  &NiagaraDataInterfaceArrayFloat3
  &NiagaraDataInterfaceCurve
  &NiagaraDataInterfaceArrayInt32
  &NiagaraDataInterfaceArrayFloat
  &NiagaraDataInterfaceSkeletalMesh
 ]
 usedDynamicInputs []
 usedModules [
  &self:SystemSpawnScript.SystemScriptSource.SystemScriptGraph.NiagaraNodeAssignment_0.SetVariables_#1
  &SystemState
 ]
}
emitters {
 FireRing_Body {
  emitterUpdateScript [
   &EmitterState(
    Inactive Response=::NewEnumerator1
    Life Cycle Mode=::NewEnumerator1
    Loop Behavior=::NewEnumerator1
    Loop Duration Mode=::NewEnumerator1
    UseLoopDelay=false
   )
   &SpawnBurst_Instantaneous(Spawn Count=2, Use Spawn Probability=false, UseLoopCountLimit=false)
  ]
  particleSpawnScript [
   &PartitionParticles()
   &InitializeParticle(
    Color=<rgba 1 0.42 0 0.7>
    Mesh Scale=&SelectVectorFromArray(
     Direct Array Index=$Particles.PartitionParticles.Partition
     Vector Selection Array=DataInterface() {propertyValues {FloatData [<1 1 1>, <-1 -1 1>]}}
    )
    Mesh Scale Mode=::NewEnumerator4
    UsePositionOffset=false
    Write Material Random=true
   )
   &InitialMeshOrientation(
    Mesh Orientation Mode=::NewEnumerator2
    Rotation=&SelectVectorFromArray(
     Direct Array Index=$Particles.PartitionParticles.Partition
     Vector Selection Array=DataInterface() {propertyValues {FloatData [<0 0.25 0>, <0 0.25 0>]}}
    )
    Use Rotation Vector=true
   )
   &CameraOffset(Camera Offset Amount=2)
  ]
  particleUpdateScript [
   &ParticleState(Kill Particles When Lifetime Has Elapsed=false, Let Infinitely Lived Particles Die When Emitter Deactivates=true)
   &SkeletalMeshLocation(
    Bone / Socket Index=19
    Bone Sampling Mode=::NewEnumerator4
    MaskSpawnGroup=false
    Sampled Position Offset=&MakeCustomVectorFromBool(
     Bool=$User.bIsFrontEndPreview
     True Vector=&SelectVectorFromArray(
      Direct Array Index=$Particles.PartitionParticles.Partition
      Vector Selection Array=DataInterface() {propertyValues {FloatData [<8 -1 0>, <5 1 0>]}}
     )
     False Vector=&SelectVectorFromArray(
      Direct Array Index=$Particles.PartitionParticles.Partition
      Vector Selection Array=DataInterface() {propertyValues {FloatData [<9 0 0>, <9 0 0>]}}
     )
    )
    Skeletal Mesh=$System.Skeletal Mesh
   )
   &SolveForcesAndVelocity(Clamp Velocity=false, Limit Acceleration=false)
   &DynamicMaterialParameters(
    Index 0 Param 1=10
    Index 0 Param 2=8
    Index 0 Param 3=1
    Index 0 Param 4=1
    Module.Index0Float0Write=true
    Module.Index0Float1Write=true
    Module.Index0Float2Write=true
    Module.Index0Float3Write=true
   )
   &ScaleColor(
    Scale Alpha=&ClampFloat(Float=&FN_FrontEndFloatModifier(
      Gameplay=&ClampFloat(Float=&FN_LOD_Distance_Remap(Start Distance=200, End Distance=300, Minimum Output=0, Maximum Output=1))
      Front End=&FN_LOD_Distance_Remap(Start Distance=150, End Distance=200, Minimum Output=1, Maximum Output=1, Clamp Output=true)
     ))
    ScaleA=true
    ScaleRGB=true
   )
  ]
  rendererClasses [&NiagaraMeshRendererProperties]
  renderers [{rendererClass=&NiagaraMeshRendererProperties, rendererIndex=0}]
 }
 FireRing_Limbs {
  emitterSpawnScript [/SetVariables(Emitter.TimeMult:NiagaraFloat=&RandomRangeFloat(Minimum=1.5, Maximum=2)) {isSetParameters=true, module=SetVariables_#2}]
  emitterUpdateScript [
   &EmitterState(Inactive Response=::NewEnumerator1, Life Cycle Mode=::NewEnumerator1, Loop Duration=1, UseLoopDelay=false)
   &SpawnRate(SpawnRate=20, Use Spawn Probability=false)
  ]
  particleSpawnScript [
   &PartitionParticles(Number Of Partitions=8)
   &InitializeParticle(
    Color Maximum=<rgba 2 2 2 1>
    Color Minimum=<rgba 0.75 0.75 0.75 1>
    Color Mode=::NewEnumerator2
    Lifetime Max=1.3
    Lifetime Mode=::NewEnumerator1
    Mesh Scale=&Multiply_Vector(
     A=&SelectVectorFromArray(
      Direct Array Index=$Particles.PartitionParticles.Partition
      Vector Selection Array=DataInterface() {propertyValues {FloatData [<1 0.8 -0.8>, <1.6 1.1 -1.1>, <1 0.8 0.8>, <1.6 1.1 -1.1>, <0.8 0.8 -0.8>, <1.1 1 -1>, <0.8 0.8 0.8>, <1.1 1 1>]}}
     )
     B=&RandomRangeVector(Minimum=<0.8 1 1>, Maximum=<1 1 1>)
    )
    Mesh Scale Mode=::NewEnumerator4
    UsePositionOffset=false
    Write Material Random=true
   )
   &InitialMeshOrientation(
    Mesh Orientation Mode=::NewEnumerator2
    Rotation=&SelectVectorFromArray(
     Direct Array Index=$Particles.PartitionParticles.Partition
     Vector Selection Array=DataInterface() {propertyValues {FloatData [<0 0 0.5>, <0 0 0.5>, <0 0 0>, <0 0 0>, <0 0 0>, <0 0 0>, <0 0 0.5>, <0 0 0.5>]}}
    )
    Use Rotation Vector=true
   )
   &InitialMeshOrientation(
    Mesh Orientation Mode=::NewEnumerator2
    Rotation=&RandomRangeVector(Minimum=<0 0 0>, Maximum=<1 0 0>)
    Use Rotation Vector=true
   ) {module=InitialMeshOrientation001}
   &InitialMeshOrientation(
    Mesh Orientation Mode=::NewEnumerator2
    Rotation=&SelectVectorFromArray(
     Direct Array Index=$Particles.PartitionParticles.Partition
     Vector Selection Array=DataInterface() {propertyValues {FloatData [<0 0 0>, <0 0.01 0>, <0 0 0>, <0 0.01 0>, <0 0 0>, <0 0 0>, <0 0 0>, <0 0 0>]}}
    )
    Use Rotation Vector=true
   ) {module=InitialMeshOrientation002}
   &CameraOffset(Camera Offset Amount=5)
  ]
  particleUpdateScript [
   &ParticleState()
   &ScaleMeshSize(Scale Factor=&MakeCustomVectorFromBool(Bool=$User.bIsFrontEndPreview, True Vector=<1 0.8 0.8>, False Vector=<1 1 1>))
   &SkeletalMeshLocation(
    Bone / Socket Index=&SelectIntFromArray(
     Direct Array Index=$Particles.PartitionParticles.Partition
     Int Selection Array=DataInterface() {propertyValues {IntData [14,10,13,9,6,2,5,1]}}
    )
    Bone Sampling Mode=::NewEnumerator4
    MaskSpawnGroup=false
    Sampled Position Offset=&SelectVectorFromArray(
     Direct Array Index=$Particles.PartitionParticles.Partition
     Vector Selection Array=DataInterface() {propertyValues {FloatData [<30 2 -2>, <40 0 -2>, <-30 -2 2>, <-40 0 1>, <-20 0 0>, <-25 0 0>, <20 0 0>, <25 0 0>]}}
    )
    Skeletal Mesh=$System.Skeletal Mesh
   )
   /SetVariables(Particles.Width:NiagaraFloat=&RandomRangeFloat(Minimum=0.4, Maximum=0.6)) {isSetParameters=true, module=SetVariables_#3}
   &DynamicMaterialParameters(
    Index 0 Param 1=&Multiply_Float(A=15, B=$Particles.Width)
    Index 0 Param 2=&Multiply_Float(A=15, B=$Particles.Width)
    Index 0 Param 3=&RandomRangeFloat(Minimum=0, Maximum=0.8)
    Index 1 Param 1=&RandomRangeFloat(Minimum=1, Maximum=2)
    Index 2 Param 1=&RandomRangeFloat(Minimum=0.7, Maximum=1.3)
    Module.Index0Param0Write=true
    Module.Index0Param1Write=true
    Module.Index0Param2Write=true
    Module.Index0Param3Write=true
    Module.Index1Param0Write=true
    Module.Index1Param1Write=true
    Module.Index1Param2Write=true
    Module.Index1Param3Write=true
    Module.Index2Param0Write=true
    Module.Index2Param1Write=true
    Module.Index2Param2Write=true
    Module.Index2Param3Write=true
    Write Parameter Index 1=true
    Write Parameter Index 2=true
   )
   &SolveForcesAndVelocity(Clamp Velocity=false, Limit Acceleration=false)
   &ScaleColor(
    Scale Alpha=&FloatFromCurve(
     FloatCurve=DataInterface() {propertyValues {Curve {keys [<at 0 0 RCIM_Cubic>,<at 0.4 1 RCIM_Cubic>,<at 1 0 RCIM_Cubic>]}}}
     Scale Curve=&RandomRangeFloat(Minimum=1, Maximum=1)
    )
    Scale RGB=&MakeVectorFromLinearColorRGB(Color=&RandomRangeLinearColor(Minimum=<rgba 1 0.167 0.02 1>, Maximum=<rgba 1 0.4493 0 1>))
    ScaleA=true
    ScaleRGB=true
   )
   &ScaleMeshSize(Scale Factor=&VectorFromFloat(Value=&MakeCustomFloatFromBool(
      Bool=$User.bisHandHidden
      True Float=&SelectFloatFromArray(
       Direct Array Index=$Particles.PartitionParticles.Partition
       Float Selection Array=DataInterface() {propertyValues {FloatData [0,1,0,1,0,1,0,1]}}
      )
      False Float=&SelectFloatFromArray(
       Direct Array Index=$Particles.PartitionParticles.Partition
       Float Selection Array=DataInterface() {propertyValues {FloatData [0,1,0,1,1,1,1,1]}}
      )
     ))) {module=ScaleMeshSize001}
  ]
  rendererClasses [&NiagaraMeshRendererProperties]
  renderers [{rendererClass=&NiagaraMeshRendererProperties, rendererIndex=0}]
 }
 FireRing_Limbs001 {
  emitterSpawnScript [/SetVariables(Emitter.TimeMult:NiagaraFloat=&RandomRangeFloat(Minimum=1.5, Maximum=2)) {isSetParameters=true, module=SetVariables_#2}]
  emitterUpdateScript [
   &EmitterState(Inactive Response=::NewEnumerator1, Life Cycle Mode=::NewEnumerator1, Loop Duration=1, UseLoopDelay=false)
   &SpawnRate(SpawnRate=20, Use Spawn Probability=false)
  ]
  particleSpawnScript [
   &PartitionParticles(Number Of Partitions=8)
   &InitializeParticle(
    Color Maximum=<rgba 2 2 2 1>
    Color Minimum=<rgba 0.75 0.75 0.75 1>
    Color Mode=::NewEnumerator2
    Lifetime Max=1.3
    Lifetime Mode=::NewEnumerator1
    Mesh Scale=&Multiply_Vector(
     A=&SelectVectorFromArray(
      Direct Array Index=$Particles.PartitionParticles.Partition
      Vector Selection Array=DataInterface() {propertyValues {FloatData [<1 0.8 -0.8>, <1.6 1.1 -1.1>, <1 0.8 0.8>, <1.6 1.1 -1.1>, <0.8 0.8 -0.8>, <1.1 1 -1>, <0.8 0.8 0.8>, <1.1 1 1>]}}
     )
     B=&RandomRangeVector(Minimum=<0.8 1 1>, Maximum=<1 1 1>)
    )
    Mesh Scale Mode=::NewEnumerator4
    UsePositionOffset=false
    Write Material Random=true
   )
   &InitialMeshOrientation(
    Mesh Orientation Mode=::NewEnumerator2
    Rotation=&SelectVectorFromArray(
     Direct Array Index=$Particles.PartitionParticles.Partition
     Vector Selection Array=DataInterface() {propertyValues {FloatData [<0 0 0.5>, <0 0 0.5>, <0 0 0>, <0 0 0>, <0 0 0>, <0 0 0>, <0 0 0.5>, <0 0 0.5>]}}
    )
    Use Rotation Vector=true
   )
   &InitialMeshOrientation(
    Mesh Orientation Mode=::NewEnumerator2
    Rotation=&RandomRangeVector(Minimum=<0 0 0>, Maximum=<1 0 0>)
    Use Rotation Vector=true
   ) {module=InitialMeshOrientation001}
   &InitialMeshOrientation(
    Mesh Orientation Mode=::NewEnumerator2
    Rotation=&SelectVectorFromArray(
     Direct Array Index=$Particles.PartitionParticles.Partition
     Vector Selection Array=DataInterface() {propertyValues {FloatData [<0 0 0>, <0 0.01 0>, <0 0 0>, <0 0.01 0>, <0 0 0>, <0 0 0>, <0 0 0>, <0 0 0>]}}
    )
    Use Rotation Vector=true
   ) {module=InitialMeshOrientation002}
   &CameraOffset(Camera Offset Amount=5)
  ]
  particleUpdateScript [
   &ParticleState()
   &ScaleMeshSize(Scale Factor=&MakeCustomVectorFromBool(Bool=$User.bIsFrontEndPreview, True Vector=<1 0.8 0.8>, False Vector=<1 1 1>))
   &SkeletalMeshLocation(
    Bone / Socket Index=&SelectIntFromArray(
     Direct Array Index=$Particles.PartitionParticles.Partition
     Int Selection Array=DataInterface() {propertyValues {IntData [14,10,13,9,6,2,5,1]}}
    )
    Bone Sampling Mode=::NewEnumerator4
    MaskSpawnGroup=false
    Sampled Position Offset=&SelectVectorFromArray(
     Direct Array Index=$Particles.PartitionParticles.Partition
     Vector Selection Array=DataInterface() {propertyValues {FloatData [<30 2 -2>, <40 0 -2>, <-30 -2 2>, <-40 0 1>, <-20 0 0>, <-25 0 0>, <20 0 0>, <25 0 0>]}}
    )
    Skeletal Mesh=$System.Skeletal Mesh
   )
   /SetVariables(Particles.Width:NiagaraFloat=&RandomRangeFloat(Minimum=0.4, Maximum=0.6)) {isSetParameters=true, module=SetVariables_#3}
   &DynamicMaterialParameters(
    Index 0 Param 1=&Multiply_Float(A=15, B=$Particles.Width)
    Index 0 Param 2=&Multiply_Float(A=15, B=$Particles.Width)
    Index 0 Param 3=&RandomRangeFloat(Minimum=0, Maximum=0.8)
    Index 1 Param 1=&RandomRangeFloat(Minimum=1, Maximum=2)
    Index 2 Param 1=&RandomRangeFloat(Minimum=0.7, Maximum=1.3)
    Module.Index0Param0Write=true
    Module.Index0Param1Write=true
    Module.Index0Param2Write=true
    Module.Index0Param3Write=true
    Module.Index1Param0Write=true
    Module.Index1Param1Write=true
    Module.Index1Param2Write=true
    Module.Index1Param3Write=true
    Module.Index2Param0Write=true
    Module.Index2Param1Write=true
    Module.Index2Param2Write=true
    Module.Index2Param3Write=true
    Write Parameter Index 1=true
    Write Parameter Index 2=true
   )
   &SolveForcesAndVelocity(Clamp Velocity=false, Limit Acceleration=false)
   &ScaleColor(
    Scale Alpha=&FloatFromCurve(
     FloatCurve=DataInterface() {propertyValues {Curve {keys [<at 0 0 RCIM_Cubic>,<at 0.4 1 RCIM_Cubic>,<at 1 0 RCIM_Cubic>]}}}
     Scale Curve=&RandomRangeFloat(Minimum=1, Maximum=1)
    )
    Scale RGB=&MakeVectorFromLinearColorRGB(Color=&RandomRangeLinearColor(Minimum=<rgba 1 0.167 0.02 1>, Maximum=<rgba 1 0.4493 0 1>))
    ScaleA=true
    ScaleRGB=true
   )
   &ScaleMeshSize(Scale Factor=&VectorFromFloat(Value=&MakeCustomFloatFromBool(
      Bool=$User.bisFootHidden
      True Float=0
      False Float=&SelectFloatFromArray(
       Direct Array Index=$Particles.PartitionParticles.Partition
       Float Selection Array=DataInterface() {propertyValues {FloatData [1,0,1,0,0,0,0,0]}}
      )
     ))) {module=ScaleMeshSize001}
  ]
  rendererClasses [&NiagaraMeshRendererProperties]
  renderers [{rendererClass=&NiagaraMeshRendererProperties, rendererIndex=0}]
 }
 Particle_Run {
  emitterUpdateScript [
   &EmitterState(Life Cycle Mode=::NewEnumerator1, Loop Duration=1, UseLoopDelay=false)
   &SpawnPerUnit(
    Max Movement Threshold=1500
    Module.Use Max Movement Threshold=true
    Module.Use Movement Tolerance=false
    Spawn Spacing:NiagaraFloat=&Multiply_Float(A=100, B=1)
    Use Spawn Probability=false
   ) {enabled=false}
   &SpawnRate(
    SpawnRate=&Multiply_Float(
     A=150
     B=&Multiply_Float(
      A=&MakeCustomFloatFromBool(
       Bool=&SetBoolByFloatComparison(A=&VectorLength(Vector=$Engine.Owner.Velocity), B=100)
       True Float=1
       False Float=0
      )
      B=&MakeCustomFloatFromBool(
       Bool=&SetBoolByFloatComparison(A=&VectorLength(Vector=$Engine.Owner.Velocity), B=1200)
       True Float=0
       False Float=1
      )
     )
    )
    Use Spawn Probability=false
   )
  ]
  particleSpawnScript [
   &FN_SetSpawnGroup(Number of Spawn Groups=20)
   &InitializeParticle(
    Color=<rgba 5 4.2184 0.7031 1>
    Lifetime Max=0.4
    Lifetime Min=0.3
    Lifetime Mode=::NewEnumerator1
    Sprite Size Max=<1.2 2.5>
    Sprite Size Min=<1 2>
    Sprite Size Mode=::NewEnumerator2
    UsePositionOffset=true
    Write Material Random=true
   )
   &ShapeLocation(MaskSpawnGroup=false, Override Seed=false, Rotation Axis=<0 1 0>, Rotation Mode=::NewEnumerator1, Sphere Radius=10, UseOwnerScale=true)
  ]
  particleUpdateScript [
   &ParticleState()
   /SetVariables(Particles.SizeCustomScale:Vector2f=&Multiply_Vector2D_ByFloat(
     Float=&FloatFromCurve(FloatCurve=DataInterface() {propertyValues {Curve {keys [<at 0 0 RCIM_Cubic>,<at 0.2 0.75 RCIM_Cubic>,<at 1 0>]}}})
     Vector2D=<1 1>
    )) {isSetParameters=true}
   &SkeletalMeshLocation(
    Bone / Socket Index=$Particles.SpawnGroup
    Bone Sampling Mode=::NewEnumerator4
    MaskSpawnGroup=false
    Orientation Sampling=::NewEnumerator1
    Position Sampling=::NewEnumerator1
    Sampled Position Offset=&AddVectorToPosition(Position=<0 0 0>, Vector=$Particles.ShapeLocation.ShapeVector)
    Skeletal Mesh=$System.Skeletal Mesh
    Source Movement Scale:Vector3f=<0.5 0.5 0.7>
   )
   &CurlNoiseForce(Bias Noise=false, Bias Noise Field:Vector3f=<0 0 0.2>, Noise Strength=150, Pan Noise=false)
   &SolveForcesAndVelocity(Clamp Velocity=false, Limit Acceleration=false)
   &ScaleColor(
    Scale Alpha=&FloatFromCurve(FloatCurve=DataInterface() {propertyValues {Curve {keys [
         {interpMode=RCIM_Cubic, tangentMode=RCTM_User, time=0, value=0}
         {interpMode=RCIM_Cubic, tangentMode=RCTM_User, time=0.5, value=1}
         {tangentMode=RCTM_User, time=1, value=0}
        ]}}})
    Scale RGB=&MakeVectorFromLinearColorRGB(Color=<rgba 1 1 1 1>)
    ScaleA=true
    ScaleRGB=true
   )
   /SetVariables(Particles.Glow03_Color=&MakeLinearColorFromVectorAndFloat(
     "Vector (RGB)"=&MakeVectorFromLinearColorRGB(Color=&MultiplyLinearColor(Linear Color=<rgba 1 0.1229 0 1>, Scale Factor=&MakeVector4_FromFloat(Float=1)))
     "Float (Alpha)"=&FloatFromCurve(FloatCurve=DataInterface() {propertyValues {Curve {keys [
          {interpMode=RCIM_Cubic, tangentMode=RCTM_User, time=0, value=0}
          {interpMode=RCIM_Cubic, tangentMode=RCTM_User, time=0.5, value=1}
          {tangentMode=RCTM_User, time=1, value=0}
         ]}}})
    )) {isSetParameters=true}
   &ScaleSpriteSize(
    Scale Factor=&Multiply_Vector2D(
     A=&Multiply_Vector2D_ByFloat(Float=&RandomRangeFloat(Minimum=0.5, Maximum=1), Vector2D=<1 1>)
     B=$Particles.SizeCustomScale
    )
    Scale Sprite Size Mode=::NewEnumerator2
   )
   /SetVariables(Particles.Glow03_Size=&Multiply_Vector2D(A=<5 7>, B=$Particles.SizeCustomScale)) {isSetParameters=true}
  ]
  rendererClasses [&NiagaraSpriteRendererProperties]
  renderers [
   {rendererClass=&NiagaraSpriteRendererProperties}
   {rendererClass=&NiagaraSpriteRendererProperties, rendererIndex=1}
  ]
 }
 RayFire_Mesh {
  emitterUpdateScript [
   &EmitterState(
    Inactive Response=::NewEnumerator1
    Life Cycle Mode=::NewEnumerator1
    Loop Behavior=::NewEnumerator1
    Loop Duration Mode=::NewEnumerator1
    UseLoopDelay=false
   )
   &SpawnBurst_Instantaneous(Spawn Count=21, Use Spawn Probability=false, UseLoopCountLimit=false)
  ]
  particleSpawnScript [
   &PartitionParticles(Number Of Partitions=21)
   &InitializeParticle(
    Color Maximum=<rgba 1 1 1 0.2>
    Color Minimum=<rgba 1 1 1 0.1>
    Color Mode=::NewEnumerator2
    Mesh Scale=&Multiply_Vector(
     A=&SelectVectorFromArray(
      Direct Array Index=$Particles.PartitionParticles.Partition
      Vector Selection Array=DataInterface() {propertyValues {FloatData [
         <0.8 0.7 15>
         <0.8 0.8 5>
         <0.5 0.8 20>
         <0.5 1 -40>
         <0.7 0.8 30>
         <0.6 1.2 -60>
         <0.6 1 30>
         <0.6 1 -40>
         <0.6 0.7 30>
         <0.6 0.75 -30>
         <0.7 1.5 -80>
         <0.7 1 60>
         <0.7 1 -50>
         <0.7 0.8 60>
         <0.6 1.5 -75>
         <0.6 1.5 80>
         <0.6 1.5 -50>
         <0.5 1 70>
         <0.5 0.8 50>
         <0.6 0.8 1>
         <0.6 0.8 1>
        ]}}
     )
     B=<0.95 0.95 1>
    )
    Mesh Scale Mode=::NewEnumerator4
    Sprite Size Mode=::NewEnumerator3
    Uniform Sprite Size=20
    UsePositionOffset=false
    Write Material Random=true
   )
   &InitialMeshOrientation(Mesh Orientation Mode=::NewEnumerator2, Rotation=<0 0 0.25>, Use Rotation Vector=true)
  ]
  particleUpdateScript [
   &ParticleState(Kill Particles When Lifetime Has Elapsed=false, Let Infinitely Lived Particles Die When Emitter Deactivates=true)
   &SkeletalMeshLocation(
    Bone / Socket Index=&SelectIntFromArray(
     Direct Array Index=$Particles.PartitionParticles.Partition
     Int Selection Array=DataInterface() {propertyValues {IntData [20,21,1,2,1,2,5,6,5,6,9,10,9,10,9,10,13,14,14,20,21]}}
    )
    Bone Sampling Mode=::NewEnumerator4
    MaskSpawnGroup=false
    Sampled Position Offset=&SelectVectorFromArray(
     Direct Array Index=$Particles.PartitionParticles.Partition
     Vector Selection Array=DataInterface() {propertyValues {FloatData [
        <15 0 0>
        <-12 0 -2>
        <5 0 3>
        <-5 0 0>
        <17 0 2>
        <-15 0 0>
        <6 0 2>
        <-6 0 -2>
        <18 0 0>
        <-18 0 0>
        <-5 0 3>
        <0 0 2>
        <-20 0 -0.5>
        <15 0 2>
        <-30 0 2>
        <23 0 -1>
        <-20 0 0>
        <13 0 -0.5>
        <28 0 0>
        <5 0 5>
        <-5 0 -4>
       ]}}
    )
    Skeletal Mesh=$System.Skeletal Mesh
   )
   &DynamicMaterialParameters(
    Index 0 Param 1=&RandomRangeFloat(Minimum=0.4, Maximum=0.8)
    Index 0 Param 2=&RandomRangeFloat(Minimum=0.4, Maximum=0.8)
    Index 0 Param 4=&RandomRangeFloat(Minimum=1, Maximum=1.3)
    Index 2 Param 1=&RandomRangeFloat(Minimum=0.8, Maximum=1)
    Index 2 Param 2=&Multiply_Float(
     A=&SelectFloatFromArray(
      Direct Array Index=$Particles.PartitionParticles.Partition
      Float Selection Array=DataInterface() {propertyValues {FloatData [1.3,1.5,0.5,0.8,1,0.8,0.85,0.8,0.75,0.6,1,0.7,1,0.5,0.95,1,0.9,1,0.8,1.5,1.5]}}
     )
     B=0.65
    )
    Module.Index0Float0Write=true
    Module.Index0Float1Write=true
    Module.Index0Float2Write=true
    Module.Index0Float3Write=true
    Module.Index2Float0Write=true
    Module.Index2Float1Write=true
    Module.Index2Float2Write=true
    Module.Index2Float3Write=true
    Write Parameter Index 2=::NewEnumerator1
   )
   &InterpolateOverTime(Rate Of Change=5, Target Vector Value:NiagaraPosition=$Particles.Position, Variable Type=::NewEnumerator2)
   /SetVariables(Particles.Interpolate Dynamic=&MakeFloat4FromVectorAndFloat("Vector (RGB)"=$Particles.InterpolateOverTime.InterpolateOverTime_V3_.Moving Average, "Float (Alpha)"=0.4)) {isSetParameters=true, module=SetVariables_#4}
   &ScaleColor(
    Scale Alpha=&MakeCustomFloatFromBool(Bool=$User.bIsFrontEndPreview, True Float=0.3, False Float=1)
    Scale RGB=&MakeVectorFromLinearColorRGB(Color=<rgba 1 1 1 1>)
    ScaleA=true
    ScaleRGB=true
   )
   &ScaleMeshSize(Scale Factor=&MakeCustomVectorFromBool(
     Bool=$User.bisHandHidden
     True Vector=&VectorFromFloat(Value=&SelectFloatFromArray(
       Direct Array Index=$Particles.PartitionParticles.Partition
       Float Selection Array=DataInterface() {propertyValues {FloatData [1,1,1,1,1,1,0,0,0,0,1,1,1,1,0,1,0,0,0,1,1]}}
      ))
     False Vector=&VectorFromFloat(Value=&SelectFloatFromArray(
       Direct Array Index=$Particles.PartitionParticles.Partition
       Float Selection Array=DataInterface() {propertyValues {FloatData [1,1,1,1,1,1,1,1,1,1,1,1,1,1,0,1,0,0,0,1,1]}}
      ))
    ))
  ]
  rendererClasses [&NiagaraMeshRendererProperties]
  renderers [{rendererClass=&NiagaraMeshRendererProperties, rendererIndex=0}]
 }
 RayFire_Mesh001 {
  emitterUpdateScript [
   &EmitterState(
    Inactive Response=::NewEnumerator1
    Life Cycle Mode=::NewEnumerator1
    Loop Behavior=::NewEnumerator1
    Loop Duration Mode=::NewEnumerator1
    UseLoopDelay=false
   )
   &SpawnBurst_Instantaneous(Spawn Count=21, Use Spawn Probability=false, UseLoopCountLimit=false)
  ]
  particleSpawnScript [
   &PartitionParticles(Number Of Partitions=21)
   &InitializeParticle(
    Color Maximum=<rgba 1 1 1 0.2>
    Color Minimum=<rgba 1 1 1 0.1>
    Color Mode=::NewEnumerator2
    Mesh Scale=&Multiply_Vector(
     A=&SelectVectorFromArray(
      Direct Array Index=$Particles.PartitionParticles.Partition
      Vector Selection Array=DataInterface() {propertyValues {FloatData [
         <0.8 0.7 15>
         <0.8 0.8 5>
         <0.5 0.8 20>
         <0.5 1 -40>
         <0.7 0.8 30>
         <0.6 1.2 -60>
         <0.6 1 30>
         <0.6 1 -40>
         <0.6 0.7 30>
         <0.6 0.75 -30>
         <0.7 1.5 -80>
         <0.7 1 60>
         <0.7 1 -50>
         <0.7 0.8 60>
         <0.6 1.5 -75>
         <0.6 1.5 80>
         <0.6 1.5 -50>
         <0.5 1 70>
         <0.5 0.8 50>
         <0.6 0.8 1>
         <0.6 0.8 1>
        ]}}
     )
     B=<0.95 0.95 1>
    )
    Mesh Scale Mode=::NewEnumerator4
    Sprite Size Mode=::NewEnumerator3
    Uniform Sprite Size=20
    UsePositionOffset=false
    Write Material Random=true
   )
   &InitialMeshOrientation(Mesh Orientation Mode=::NewEnumerator2, Rotation=<0 0 0.25>, Use Rotation Vector=true)
  ]
  particleUpdateScript [
   &ParticleState(Kill Particles When Lifetime Has Elapsed=false, Let Infinitely Lived Particles Die When Emitter Deactivates=true)
   &SkeletalMeshLocation(
    Bone / Socket Index=&SelectIntFromArray(
     Direct Array Index=$Particles.PartitionParticles.Partition
     Int Selection Array=DataInterface() {propertyValues {IntData [20,21,1,2,1,2,5,6,5,6,9,10,9,10,9,10,13,14,14,20,21]}}
    )
    Bone Sampling Mode=::NewEnumerator4
    MaskSpawnGroup=false
    Sampled Position Offset=&SelectVectorFromArray(
     Direct Array Index=$Particles.PartitionParticles.Partition
     Vector Selection Array=DataInterface() {propertyValues {FloatData [
        <15 0 0>
        <-12 0 -2>
        <5 0 3>
        <-5 0 0>
        <17 0 2>
        <-15 0 0>
        <6 0 2>
        <-6 0 -2>
        <18 0 0>
        <-18 0 0>
        <-5 0 3>
        <0 0 2>
        <-20 0 -0.5>
        <15 0 2>
        <-30 0 2>
        <23 0 -1>
        <-20 0 0>
        <13 0 -0.5>
        <28 0 0>
        <5 0 5>
        <-5 0 -4>
       ]}}
    )
    Skeletal Mesh=$System.Skeletal Mesh
   )
   &DynamicMaterialParameters(
    Index 0 Param 1=&RandomRangeFloat(Minimum=0.4, Maximum=0.8)
    Index 0 Param 2=&RandomRangeFloat(Minimum=0.4, Maximum=0.8)
    Index 0 Param 4=&RandomRangeFloat(Minimum=1, Maximum=1.3)
    Index 2 Param 1=&RandomRangeFloat(Minimum=0.8, Maximum=1)
    Index 2 Param 2=&Multiply_Float(
     A=&SelectFloatFromArray(
      Direct Array Index=$Particles.PartitionParticles.Partition
      Float Selection Array=DataInterface() {propertyValues {FloatData [1.3,1.5,0.5,0.8,1,0.8,0.85,0.8,0.75,0.6,1,0.7,1,0.5,0.95,1,0.9,1,0.8,1.5,1.5]}}
     )
     B=0.65
    )
    Module.Index0Float0Write=true
    Module.Index0Float1Write=true
    Module.Index0Float2Write=true
    Module.Index0Float3Write=true
    Module.Index2Float0Write=true
    Module.Index2Float1Write=true
    Module.Index2Float2Write=true
    Module.Index2Float3Write=true
    Write Parameter Index 2=::NewEnumerator1
   )
   &InterpolateOverTime(Rate Of Change=5, Target Vector Value:NiagaraPosition=$Particles.Position, Variable Type=::NewEnumerator2)
   /SetVariables(Particles.Interpolate Dynamic=&MakeFloat4FromVectorAndFloat("Vector (RGB)"=$Particles.InterpolateOverTime.InterpolateOverTime_V3_.Moving Average, "Float (Alpha)"=0.4)) {isSetParameters=true, module=SetVariables_#4}
   &ScaleColor(
    Scale Alpha=&MakeCustomFloatFromBool(Bool=$User.bIsFrontEndPreview, True Float=0.3, False Float=1)
    Scale RGB=&MakeVectorFromLinearColorRGB(Color=<rgba 1 1 1 1>)
    ScaleA=true
    ScaleRGB=true
   )
   &ScaleMeshSize(Scale Factor=&MakeCustomVectorFromBool(
     Bool=$User.bisFootHidden
     True Vector=&VectorFromFloat(Value=0)
     False Vector=&VectorFromFloat(Value=&SelectFloatFromArray(
       Direct Array Index=$Particles.PartitionParticles.Partition
       Float Selection Array=DataInterface() {propertyValues {FloatData [0,0,0,0,0,0,0,0,0,0,0,0,0,0,1,0,1,1,1,0,0]}}
      ))
    ))
  ]
  rendererClasses [&NiagaraMeshRendererProperties]
  renderers [{rendererClass=&NiagaraMeshRendererProperties, rendererIndex=0}]
 }
 RayFire_sphere {
  emitterUpdateScript [
   &EmitterState(
    Inactive Response=::NewEnumerator1
    Life Cycle Mode=::NewEnumerator1
    Loop Behavior=::NewEnumerator1
    Loop Duration Mode=::NewEnumerator1
    UseLoopDelay=false
   )
   &SpawnBurst_Instantaneous(Spawn Count=19, Use Spawn Probability=false, UseLoopCountLimit=false)
  ]
  particleSpawnScript [
   &PartitionParticles(Number Of Partitions=20)
   &InitializeParticle(
    Lifetime=3.6
    Mesh Scale=&Multiply_VectorByFloat(Vector=<0.3 1 0.3>, Float=&RandomRangeFloat(Minimum=0.7, Maximum=0.8))
    Mesh Scale Mode=::NewEnumerator4
    Sprite Rotation Angle=-0.25
    Sprite Rotation Mode=::NewEnumerator1
    Sprite Size Mode=::NewEnumerator3
    Uniform Sprite Size=&Multiply_Float(
     A=&SelectFloatFromArray(
      Direct Array Index=$Particles.PartitionParticles.Partition
      Float Selection Array=DataInterface() {propertyValues {FloatData [0,25,25,20,20,20,20,18,18,20,20,30,30,25,25,25,25,20,20]}}
     )
     B=0.8
    )
    UsePositionOffset=false
    Write Material Random=true
   )
   /SetVariables(Particles.Random=&RandomRangeFloat(Minimum=0, Maximum=1)) {isSetParameters=true, module=SetVariables_#5}
   &InitialMeshOrientation(Mesh Orientation Mode=::NewEnumerator2, Rotation=<0 0 0.25>, Use Rotation Vector=true)
   &InitialMeshOrientation(Mesh Orientation Mode=::NewEnumerator2, Rotation=<0.25 0 0>, Use Rotation Vector=true) {module=InitialMeshOrientation001}
   &CameraOffset(Camera Offset Amount=&Multiply_Float(
     A=&SelectFloatFromArray(
      Direct Array Index=$Particles.PartitionParticles.Partition
      Float Selection Array=DataInterface() {propertyValues {FloatData [2,8,8,8,8,5,2,1,1,1,1,10,10,10,10,8,8,8,8]}}
     )
     B=0.7
    ))
  ]
  particleUpdateScript [
   &ParticleState(Kill Particles When Lifetime Has Elapsed=false, Let Infinitely Lived Particles Die When Emitter Deactivates=true)
   &SkeletalMeshLocation(
    Bone / Socket Index=&SelectIntFromArray(
     Direct Array Index=$Particles.PartitionParticles.Partition
     Int Selection Array=DataInterface() {propertyValues {IntData [0,1,2,1,2,5,5,7,8,6,6,9,10,9,10,13,14,13,14]}}
    )
    Bone Sampling Mode=::NewEnumerator4
    MaskSpawnGroup=false
    Sampled Position Offset=&SelectVectorFromArray(
     Direct Array Index=$Particles.PartitionParticles.Partition
     Vector Selection Array=DataInterface() {propertyValues {FloatData [
        <0 0 0>
        <0 0 0>
        <0 0 0>
        <13 0 0>
        <-13 0 0>
        <0 0 0>
        <0 0 0>
        <0 0 0>
        <0 0 0>
        <0 0 0>
        <0 0 0>
        <0 0 0>
        <0 0 0>
        <-20 0 0>
        <20 0 0>
        <0 0 0>
        <0 0 0>
        <-20 0 0>
        <20 0 0>
       ]}}
    )
    Skeletal Mesh=$System.Skeletal Mesh
   )
   &SolveForcesAndVelocity(Clamp Velocity=false, Limit Acceleration=false)
   &ScaleColor(
    Scale Alpha=0.5
    Scale RGB=&MakeVectorFromLinearColorRGB(Color=&RandomRangeLinearColor(Minimum=<rgba 1 0.4277 0.1651 1>, Maximum=<rgba 1 0.6637 0.2813 1>))
    ScaleA=true
    ScaleRGB=true
   )
   &ScaleSpriteSize(
    Scale Sprite Size Mode=::NewEnumerator0
    Uniform Scale Factor=&MakeCustomFloatFromBool(
     Bool=$User.bisHandHidden
     True Float=&SelectFloatFromArray(
      Direct Array Index=$Particles.PartitionParticles.Partition
      Float Selection Array=DataInterface() {propertyValues {FloatData [1,1,1,1,1,0,0,0,0,0,0,1,1,1,1,0,0,0,0]}}
     )
     False Float=&SelectFloatFromArray(
      Direct Array Index=$Particles.PartitionParticles.Partition
      Float Selection Array=DataInterface() {propertyValues {FloatData [1,1,1,1,1,1,1,1,1,1,1,1,1,1,1,0,0,0,0]}}
     )
    )
   )
  ]
  rendererClasses [&NiagaraSpriteRendererProperties]
  renderers [{rendererClass=&NiagaraSpriteRendererProperties}]
 }
 RayFire_sphere001 {
  emitterUpdateScript [
   &EmitterState(
    Inactive Response=::NewEnumerator1
    Life Cycle Mode=::NewEnumerator1
    Loop Behavior=::NewEnumerator1
    Loop Duration Mode=::NewEnumerator1
    UseLoopDelay=false
   )
   &SpawnBurst_Instantaneous(Spawn Count=19, Use Spawn Probability=false, UseLoopCountLimit=false)
  ]
  particleSpawnScript [
   &PartitionParticles(Number Of Partitions=20)
   &InitializeParticle(
    Lifetime=3.6
    Mesh Scale=&Multiply_VectorByFloat(Vector=<0.3 1 0.3>, Float=&RandomRangeFloat(Minimum=0.7, Maximum=0.8))
    Mesh Scale Mode=::NewEnumerator4
    Sprite Rotation Angle=-0.25
    Sprite Rotation Mode=::NewEnumerator1
    Sprite Size Mode=::NewEnumerator3
    Uniform Sprite Size=&Multiply_Float(
     A=&SelectFloatFromArray(
      Direct Array Index=$Particles.PartitionParticles.Partition
      Float Selection Array=DataInterface() {propertyValues {FloatData [0,25,25,20,20,20,20,18,18,20,20,30,30,25,25,25,25,20,20]}}
     )
     B=0.8
    )
    UsePositionOffset=false
    Write Material Random=true
   )
   /SetVariables(Particles.Random=&RandomRangeFloat(Minimum=0, Maximum=1)) {isSetParameters=true, module=SetVariables_#5}
   &InitialMeshOrientation(Mesh Orientation Mode=::NewEnumerator2, Rotation=<0 0 0.25>, Use Rotation Vector=true)
   &InitialMeshOrientation(Mesh Orientation Mode=::NewEnumerator2, Rotation=<0.25 0 0>, Use Rotation Vector=true) {module=InitialMeshOrientation001}
   &CameraOffset(Camera Offset Amount=&Multiply_Float(
     A=&SelectFloatFromArray(
      Direct Array Index=$Particles.PartitionParticles.Partition
      Float Selection Array=DataInterface() {propertyValues {FloatData [2,8,8,8,8,5,2,1,1,1,1,10,10,10,10,8,8,8,8]}}
     )
     B=0.7
    ))
  ]
  particleUpdateScript [
   &ParticleState(Kill Particles When Lifetime Has Elapsed=false, Let Infinitely Lived Particles Die When Emitter Deactivates=true)
   &SkeletalMeshLocation(
    Bone / Socket Index=&SelectIntFromArray(
     Direct Array Index=$Particles.PartitionParticles.Partition
     Int Selection Array=DataInterface() {propertyValues {IntData [0,1,2,1,2,5,5,7,8,6,6,9,10,9,10,13,14,13,14]}}
    )
    Bone Sampling Mode=::NewEnumerator4
    MaskSpawnGroup=false
    Sampled Position Offset=&SelectVectorFromArray(
     Direct Array Index=$Particles.PartitionParticles.Partition
     Vector Selection Array=DataInterface() {propertyValues {FloatData [
        <0 0 0>
        <0 0 0>
        <0 0 0>
        <13 0 0>
        <-13 0 0>
        <0 0 0>
        <0 0 0>
        <0 0 0>
        <0 0 0>
        <0 0 0>
        <0 0 0>
        <0 0 0>
        <0 0 0>
        <-20 0 0>
        <20 0 0>
        <0 0 0>
        <0 0 0>
        <-20 0 0>
        <20 0 0>
       ]}}
    )
    Skeletal Mesh=$System.Skeletal Mesh
   )
   &SolveForcesAndVelocity(Clamp Velocity=false, Limit Acceleration=false)
   &ScaleColor(
    Scale Alpha=0.5
    Scale RGB=&MakeVectorFromLinearColorRGB(Color=&RandomRangeLinearColor(Minimum=<rgba 1 0.4277 0.1651 1>, Maximum=<rgba 1 0.6637 0.2813 1>))
    ScaleA=true
    ScaleRGB=true
   )
   &ScaleSpriteSize(
    Scale Sprite Size Mode=::NewEnumerator0
    Uniform Scale Factor=&MakeCustomFloatFromBool(
     Bool=$User.bisFootHidden
     True Float=0
     False Float=&SelectFloatFromArray(
      Direct Array Index=$Particles.PartitionParticles.Partition
      Float Selection Array=DataInterface() {propertyValues {FloatData [0,0,0,0,0,0,0,0,0,0,0,0,0,0,0,1,1,1,1]}}
     )
    )
   )
  ]
  rendererClasses [&NiagaraSpriteRendererProperties]
  renderers [{rendererClass=&NiagaraSpriteRendererProperties}]
 }
 light {
  emitterUpdateScript [
   &EmitterState(
    Inactive Response=::NewEnumerator1
    Life Cycle Mode=::NewEnumerator1
    Loop Behavior=::NewEnumerator1
    Loop Duration Mode=::NewEnumerator1
    UseLoopDelay=false
   )
   &SpawnBurst_Instantaneous(Spawn Count=3, Use Spawn Probability=false, UseLoopCountLimit=false)
  ]
  particleSpawnScript [
   &PartitionParticles(Number Of Partitions=3)
   &InitializeParticle(
    Mesh Scale Mode=::NewEnumerator3
    Mesh Uniform Scale=2
    Sprite Size Mode=::NewEnumerator3
    Uniform Sprite Size=25
    UsePositionOffset=false
    Write Material Random=true
   )
   &CameraOffset(Camera Offset Amount=30)
  ]
  particleUpdateScript [
   &ParticleState(Kill Particles When Lifetime Has Elapsed=false, Let Infinitely Lived Particles Die When Emitter Deactivates=true)
   &ScaleColor(Scale RGB=&MakeVectorFromLinearColorRGB(Color=<rgba 3 0.39 0 1>), ScaleA=true, ScaleRGB=true) {enabled=false}
   &FN_Light_Attributes(
    Color=<rgba 4 1.6 0 1>
    Create Color Attribute=true
    Create Exponent Attribute=true
    Create Position Attribute=true
    Create Radius Attribute=true
    Enable Particle Light Enable=true
    Exponent=2
    Position=&SelectVectorFromArray(
     Direct Array Index=$Particles.PartitionParticles.Partition
     Vector Selection Array=DataInterface() {propertyValues {FloatData [<0 0 110>, <0 0 80>, <0 0 50>]}}
    )
    Radius=150
    Scale Color=2
   )
  ]
  rendererClasses [&NiagaraSpriteRendererProperties, &NiagaraLightRendererProperties]
  renderers [
   {rendererClass=&NiagaraSpriteRendererProperties}
   {rendererClass=&NiagaraLightRendererProperties, rendererIndex=1}
  ]
 }
}
summary {
 systemName=NS_Character_EggnogFaucet_BodyFire
 emitters [
  {emitterName=FireRing_Body, rendererClasses [&NiagaraMeshRendererProperties]}
  {emitterName=Particle_Run, rendererClasses [&NiagaraSpriteRendererProperties]}
  {emitterName=FireRing_Limbs, rendererClasses [&NiagaraMeshRendererProperties]}
  {emitterName=RayFire_sphere, rendererClasses [&NiagaraSpriteRendererProperties]}
  {emitterName=RayFire_Mesh, rendererClasses [&NiagaraMeshRendererProperties]}
  {emitterName=light, rendererClasses [&NiagaraSpriteRendererProperties, &NiagaraLightRendererProperties]}
  {emitterName=FireRing_Limbs001, rendererClasses [&NiagaraMeshRendererProperties]}
  {emitterName=RayFire_sphere001, rendererClasses [&NiagaraSpriteRendererProperties]}
  {emitterName=RayFire_Mesh001, rendererClasses [&NiagaraMeshRendererProperties]}
 ]
 userVariables [
  User.bisFootHidden {type=NiagaraBool}
  User.bIsFrontEnd {type=NiagaraBool}
  User.bIsFrontEndPreview {type=NiagaraBool}
  User.bisHandHidden {type=NiagaraBool}
 ]
}
