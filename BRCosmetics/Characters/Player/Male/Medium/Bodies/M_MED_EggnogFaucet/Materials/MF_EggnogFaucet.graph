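# Material function for the EggnogFaucet character body that layers an animated, flickering flame effect over incoming material attributes.
# Panned mask-texture noise, world-aligned noise on pre-skinned coordinates and Fresnel terms build two-color flames with freckles and accents.
# TechArtMask channels, an optional chest sphere mask and an EffectMask-driven radial reveal control placement.
# Outputs blend emissive, darken base and subsurface color, alter opacity, and optionally offset vertices along normals.
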MaterialFunction /BRCosmetics/Characters/Player/Male/Medium/Bodies/M_MED_EggnogFaucet/Materials/MF_EggnogFaucet
graph {nodes [
  CameraVectorWS()
  n1 = GetMaterialAttributes(None=FunctionInput() {inputType=FunctionInput_MaterialAttributes}) {attributeGetTypes [
    #1
    #2
    #3
    #4
   ]}
  n2 = SetMaterialAttributes!(
   MaterialAttributes=@n1.MaterialAttributes
   EmissiveColor=NamedRerouteUsage()
   BaseColor=NamedRerouteUsage()
   SubsurfaceColor=NamedRerouteUsage()
   WorldPositionOffset=NamedRerouteUsage()
   Opacity=NamedRerouteUsage()
  ) {attributeSetTypes [
    #1
    #2
    #3
    #5
    #4
   ]}
  FunctionOutput(None=@n2)
  n3 = Add(A=NamedRerouteUsage(), B=VectorParameter() {parameterName=RadialNoisePosition, defaultValue=<rgba 0 0 -126.8665 1>}.RGB)
  n4 = MaterialFunctionCall(Vector or UVs=Multiply(A=ComponentMask(None=@n3) {b=true, r=true}) {constB=0.02}) {materialFunction=&VectorToRadialValue}
  n5 = Multiply(A=@n4.Radial Coordinates, B=Constant2Vector() {g=1.5, r=2})
  n6 = Multiply(A=ComponentMask(None=NamedRerouteUsage()) {g=true, r=true}) {constB=0.25}
  n7 = TextureSample(UVs=Panner(Coordinate=Add(A=@n5, B=@n6)) {speedX=0.05, speedY=-1}) {
   samplerType=SAMPLERTYPE_LinearColor
   texture=&/T_EggnogFaucet_mask
  }
  Max(A=@n7.G, B=@n4.Linear Distance)
  Min(A=@n7.G, B=@n4.Linear Distance)
  n8 = NamedRerouteUsage()
  n9 = VectorParameter() {parameterName=MaskPositionOffset, defaultValue=<rgba 0 -15.8219 -130.9858 1>}
  n10 = SphereMask(
   A=Multiply(A=Add(A=@n8, B=@n9.RGB), B=VectorParameter() {parameterName=MaskScale, defaultValue=<rgba 1.0906 1 0.8475 1>}.RGB)
   B=Constant()
   Radius=ScalarParameter() {defaultValue=35.5879, parameterName=MaskRadius}
   Hardness=ScalarParameter() {parameterName=MaskHardness}
  ) {attenuationRadius=30, hardnessPercent=0}
  n11 = ScalarParameter() {defaultValue=1.3, parameterName=MaskIntensity}
  n12 = Constant() {r=0.05}
  n13 = Multiply(
   A=Saturate(None=MaterialFunctionCall!(FlickerInitialSpeed=@n12, FlickerRandom=@n12) {materialFunction=/Game/Effects/Fort_Effects/Materials/_Functions/MF_Flicker}.Result)
   B=ScalarParameter() {defaultValue=0.3, parameterName=MaskFlickerIntensity}
  ) {constB=0.3}
  n14 = Saturate(None=Multiply(A=@n10, B=Add(A=@n11, B=@n13)))
  n15 = Multiply(A=NamedRerouteUsage()) {constB=0.25}
  n16 = LinearInterpolate(A=Multiply(A=Subtract(A=@n14, B=@n15), B=@n14), Alpha=SmoothStep(Value=@n10) {constMax=0.8, constMin=0.75})
  NamedRerouteDeclaration(None=StaticSwitchParameter(True=Saturate(None=@n16), False=Constant()) {parameterName=UseBodyMask}) {name=ChestMask, nodeColor=<rgba 0 1 0.25 1>}
  NamedRerouteDeclaration(None=@n1.Opacity) {name=InOpacity, nodeColor=<rgba 0.6406 0 1 1>}
  n17 = VertexNormalWS()
  n18 = NamedRerouteUsage()
  n19 = Multiply(A=@n18, B=@n18)
  n20 = ComponentMask(None=VertexNormalWS()) {b=true}
  n21 = Multiply(
   A=Multiply(A=LinearInterpolate(Alpha=@n19) {constA=-0.1}, B=ScalarParameter() {defaultValue=1.5, parameterName=WPOIntensity}) {constB=3}
   B=LinearInterpolate(B=ConstantBiasScale(None=@n20), Alpha=ScalarParameter() {parameterName=WPOUseZNormalMask}) {constA=1}
  )
  n22 = Multiply(A=NamedRerouteUsage()) {constB=0.01}
  n23 = Abs(None=Multiply(A=ComponentMask(None=@n22) {r=true}) {constB=1.8})
  n24 = OneMinus(None=ComponentMask(None=@n22) {b=true})
  n25 = Saturate(None=Max(A=SmoothStep(Value=@n23) {constMin=0.5}, B=Power(Base=@n24) {constExponent=25}))
  n26 = Constant()
  n27 = LinearInterpolate(
   A=LinearInterpolate(A=Multiply(A=@n17, B=@n21), Alpha=StaticSwitchParameter(True=@n25, False=@n26) {parameterName=UseBodyMask}) {constB=0}
   Alpha=NamedRerouteUsage()
  ) {constB=0}
  n28 = Constant()
  n29 = NamedRerouteUsage()
  NamedRerouteDeclaration(None=Multiply(A=StaticSwitchParameter(True=@n27, False=@n28) {parameterName=UseWPO}, B=@n29)) {name=OutWPO, nodeColor=<rgba 0.5 0 1 1>}
  n30 = TextureObject() {
   samplerType=SAMPLERTYPE_LinearColor
   texture=&T_EggnogFaucet_mask
  }
  n31 = ScalarParameter() {defaultValue=20, parameterName=TextureScale}
  n32 = Multiply(A=Multiply(A=Time()) {constB=25}, B=VectorParameter() {parameterName=TextureSpeed, defaultValue=<rgba 0 0 -1 1>}.RGB)
  n33 = NamedRerouteUsage()
  n34 = Power(Base=NamedRerouteUsage())
  n35 = Multiply(A=NamedRerouteUsage(), B=Multiply(A=Subtract(A=@n34) {constB=0.5}) {constB=2})
  n36 = NamedRerouteUsage()
  n37 = Add(
   A=Add(A=Add(A=@n32, B=@n33), B=Multiply(A=@n35) {constB=2})
   B=Multiply(
    A=MaterialFunctionCall(Float2=@n36) {materialFunction=&BreakOutFloat2Components}.R
    B=ScalarParameter() {defaultValue=1, parameterName=NormalDistortionAmount}
   )
  )
  n38 = PreSkinnedNormal()
  n39 = NamedRerouteDeclaration(None=ShaderStageSwitch(PixelShader=VertexInterpolator(VS=@n38).PS, VertexShader=@n38)) {name=PreSkinnedNormal, nodeColor=<rgba 0 0.5078 1 1>}
  NamedRerouteDeclaration(None=MaterialFunctionCall(Float3=MaterialFunctionCall(TextureObject=@n30, TextureSize=@n31, WorldPosition=@n37, World Space Normal=@n39) {materialFunction=&WorldAlignedTexture}.XYZ Texture) {materialFunction=&BreakOutFloat3Components}.G) {name=Noise2, nodeColor=<rgba 1 0.2344 0 1>}
  n40 = ComponentMask(None=NamedRerouteUsage()) {b=true}
  n41 = Time()
  n42 = Sine(None=Add(A=Multiply(A=@n40) {constB=-0.005}, B=Multiply(A=@n41) {constB=0.3}))
  NamedRerouteDeclaration(None=Power(Base=ConstantBiasScale(None=@n42)) {constExponent=5}) {name=Wipe, nodeColor=<rgba 0 1 0.9297 1>}
  n43 = MaterialFunctionCall(Float3=NamedRerouteUsage()) {materialFunction=&BreakOutFloat3Components}
  n44 = OneMinus(None=Max(A=Abs(None=@n43.R), B=Abs(None=@n43.B)))
  n45 = Multiply(A=ComponentMask(None=NamedRerouteUsage()) {b=true, r=true}) {constB=0.05}
  n46 = ComponentMask(None=VectorParameter() {parameterName=DistorionNoiseSpeed, defaultValue=<rgba 0.1 0.5 0 1>}.RGB) {g=true, r=true}
  n47 = TextureSample(UVs=Multiply(A=Panner(Coordinate=@n45, Speed=@n46) {speedY=0.5}) {constB=0.5}) {
   samplerType=SAMPLERTYPE_LinearColor
   texture=&T_EggnogFaucet_mask
  }
  n48 = Time()
  n49 = Sine(None=Add(A=@n47.B, B=Multiply(A=@n48) {constB=0.5}))
  n50 = Add(A=Multiply(A=@n49) {constB=0.05}, B=@n45)
  n51 = Multiply(A=@n46) {constB=-3.5}
  n52 = TextureSample(UVs=Multiply(A=Panner(Coordinate=@n50, Speed=@n51)) {constB=0.3}) {
   samplerType=SAMPLERTYPE_LinearColor
   texture=&T_EggnogFaucet_mask
  }
  n53 = Multiply(A=@n52.R, B=NamedRerouteUsage())
  n54 = Time()
  n55 = Sine(None=Add(A=NamedRerouteDeclaration(None=@n52.B) {name=MedFreqNoise, nodeColor=<rgba 0.3125 1 0 1>}, B=Multiply(A=@n54) {constB=0.5}))
  n56 = NamedRerouteUsage()
  n57 = NamedRerouteDeclaration(None=Multiply(A=ConstantBiasScale(None=@n55), B=Power(Base=@n56))) {name=MedFreqNoiseAnim, nodeColor=<rgba 1 0 0.7266 1>}
  NamedRerouteDeclaration(None=Multiply(A=SmoothStep(Value=@n44), B=Multiply(A=@n53, B=@n57))) {name=Freckles, nodeColor=<rgba 0 1 0.6953 1>}
  n58 = Constant() {r=0.1}
  n59 = Multiply(A=NamedRerouteDeclaration(None=MaterialFunctionCall(FlickerInitialSpeed=@n58, FlickerRandom=@n58) {materialFunction=&MF_Flicker}.Result) {name=Flicker, nodeColor=<rgba 0 1 0.3906 1>}) {constB=0.4}
  n60 = ScalarParameter() {defaultValue=1, parameterName=InnerFresnelExponent}
  n61 = Multiply(A=@n60) {constB=1.5}
  n62 = NamedRerouteUsage()
  n63 = Fresnel(ExponentIn=Add(A=@n59, B=LinearInterpolate(A=@n61, B=@n60, Alpha=@n62))) {baseReflectFraction=0, exponent=1.5}
  n64 = NamedRerouteUsage()
  n65 = NamedRerouteUsage()
  n66 = Multiply(A=LinearInterpolate(Alpha=@n63) {constA=0.8}, B=Multiply(A=Multiply(A=@n64, B=@n64), B=@n65))
  n67 = LinearInterpolate(Alpha=Power(Base=NamedRerouteUsage()) {constExponent=3}) {constA=0.8, constB=1.2}
  n68 = Multiply(A=Add(A=@n65, B=@n63), B=@n63)
  n69 = LinearInterpolate(Alpha=Power(Base=OneMinus(None=@n63)) {constExponent=1.5}) {constA=0.2}
  n70 = Add(A=Add(A=NamedRerouteUsage(), B=Multiply(A=@n66, B=@n67)), B=Multiply(A=Multiply(A=@n68, B=@n69), B=@n67) {constB=1.5})
  n71 = ScalarParameter() {defaultValue=1, parameterName=BaseNoiseIntensity}
  n72 = Add(A=@n59, B=ScalarParameter() {defaultValue=3, parameterName=OutterFresnelExponent})
  n73 = Multiply(
   A=@n63
   B=OneMinus(None=MaterialFunctionCall(Exponent=@n72) {materialFunction=&AxisAlignedFresnel}.Fresnel)
  )
  n74 = NamedRerouteUsage()
  n75 = Multiply(A=@n73, B=Add(A=@n73, B=LinearInterpolate(Alpha=@n74) {constA=-0.5}))
  n76 = Power(Base=NamedRerouteUsage()) {constExponent=3}
  n77 = Multiply(A=Multiply(A=SmoothStep(Value=@n75), B=LinearInterpolate(Alpha=@n76) {constA=0.1}), B=OneMinus(None=NamedRerouteUsage()))
  n78 = ScalarParameter() {defaultValue=10, parameterName=EdgesNoiseIntensity}
  n79 = Add(
   A=Multiply(A=NamedRerouteUsage(), B=ScalarParameter() {defaultValue=50, parameterName=FrecklesIntensity}) {constB=50}
   B=Add(A=Multiply(A=@n70, B=@n71), B=Multiply(A=@n77, B=@n78) {constB=2})
  )
  n80 = ComponentMask(None=VertexNormalWS()) {b=true}
  n81 = LinearInterpolate(Alpha=Power(Base=ConstantBiasScale(None=@n80)) {constExponent=8}) {constA=1, constB=2}
  n82 = NamedRerouteUsage()
  n83 = LinearInterpolate(
   A=VectorParameter() {parameterName=Color1, defaultValue=<rgba 0.8594 0.0224 0 1>}.RGB
   B=VectorParameter() {parameterName=Color2, defaultValue=<rgba 2 1.5631 0 1>}.RGB
   Alpha=Multiply(A=Multiply(A=@n79, B=@n81), B=LinearInterpolate(Alpha=@n82) {constA=0.8, constB=1.5})
  )
  n84 = VectorParameter() {parameterName=AccentColor, defaultValue=<rgba 2 1.7542 0.875 1>}
  n85 = NamedRerouteUsage()
  n86 = NamedRerouteUsage()
  NamedRerouteDeclaration(None=Multiply(A=LinearInterpolate(A=@n83, B=@n84.RGB, Alpha=@n85), B=Max(B=@n86) {constA=1})) {name=Flames, nodeColor=<rgba 1 0.375 0 1>}
  n87 = NamedRerouteUsage()
  n88 = Saturate(None=NamedRerouteDeclaration(None=ScalarParameter() {parameterName=EffectMask}) {name=InEffectMask, nodeColor=<rgba 1 0.0938 0 1>})
  n89 = SphereMask(A=@n3, B=Constant(), Radius=Multiply(A=@n88) {constB=400}) {attenuationRadius=500, hardnessPercent=0}
  n90 = SmoothStep(Min=Saturate(None=Multiply(A=@n87, B=@n87)), Value=@n89)
  n91 = Multiply(A=@n88, B=SmoothStep(Value=@n89))
  n92 = StaticSwitchParameter(True=Saturate(None=Max(A=@n90, B=@n91)), False=@n88) {defaultValue=true, parameterName=UseRadialMasking}
  n93 = NamedRerouteUsage()
  NamedRerouteDeclaration(None=Max(A=Multiply(A=@n92, B=@n93), B=ScalarParameter() {parameterName=EffectOn})) {name=RadialEffectMask, nodeColor=<rgba 0.3828 0 1 1>}
  NamedRerouteDeclaration(None=@n1.EmissiveColor) {name=InEmissive, nodeColor=<rgba 0 0.1562 1 1>}
  n94 = FunctionInput() {inputName=TechArtMask, sortPriority=1}
  n95 = ChannelMaskParameter(None=@n94) {parameterName=FlamesEffectMask}
  n96 = ScalarParameter() {parameterName=UseFXMask}
  NamedRerouteDeclaration(None=Max(A=LinearInterpolate(B=@n95.R, Alpha=@n96) {constA=1}, B=NamedRerouteUsage())) {name=FlamesMask, nodeColor=<rgba 1 0.4219 0 1>}
  NamedRerouteDeclaration(None=@n49) {name=AnimatedPerlim, nodeColor=<rgba 1 0.6563 0 1>}
  NamedRerouteDeclaration(None=@n47.B) {name=DownGaussianNoise, nodeColor=<rgba 0.3359 0 1 1>}
  n97 = TextureSample(UVs=Panner(Coordinate=Multiply(A=@n45) {constB=0.1}) {speedX=0.1, speedY=-0.25}) {
   samplerType=SAMPLERTYPE_LinearColor
   texture=&T_EggnogFaucet_mask
  }
  n98 = Multiply(A=Time()) {constB=0.2}
  NamedRerouteDeclaration(None=Sine(None=Add(A=@n97.B, B=@n98))) {name=LowFreqGausNoiseAnim, nodeColor=<rgba 0 1 0.0156 1>}
  NamedRerouteDeclaration(None=@n52.G) {name=MedFreqVoronoi, nodeColor=<rgba 0.6406 0 1 1>}
  NamedRerouteDeclaration(None=LinearInterpolate(
    B=ChannelMaskParameter(None=@n94) {maskChannel=Green, parameterName=FlamesAccentMask, defaultValue=<rgba 0 1 0 0>}.R
    Alpha=ScalarParameter() {parameterName=UseFXAccentMask}
   )) {name=FlameAccentMask, nodeColor=<rgba 1 0 0.9141 1>}
  n99 = DotProduct(
   A=Transform(None=Constant3Vector() {constant=<rgba 0 1 0 0>}) {transformSourceType=TRANSFORMSOURCE_Local}
   B=MaterialFunctionCall() {materialFunction=&CameraDirectionVector}.Result
  )
  NamedRerouteDeclaration(None=Power(Base=Abs(None=@n99)) {constExponent=6}) {name=FresnelModulator, nodeColor=<rgba 0 1 0.9063 1>}
  NamedRerouteDeclaration(None=Multiply(A=@n7.G, B=@n4.Linear Distance)) {name=RadialVoronoi, nodeColor=<rgba 0.8984 0 1 1>}
  NamedRerouteDeclaration(None=LinearInterpolate(A=NamedRerouteUsage(), Alpha=@n29) {constB=0}) {name=OutOpacity, nodeColor=<rgba 1 0 0 1>}
  n100 = TextureSample(UVs=Multiply(A=@n45) {constB=0.5}) {
   samplerType=SAMPLERTYPE_LinearColor
   texture=&T_EggnogFaucet_mask
  }
  NamedRerouteDeclaration(None=@n100.B) {name=StaticPerlin, nodeColor=<rgba 0.4297 1 0 1>}
  n101 = NamedRerouteUsage()
  n102 = Constant()
  n103 = NamedRerouteUsage()
  n104 = LinearInterpolate(
   A=NamedRerouteUsage()
   B=Multiply(A=MaterialFunctionCall(Color_Morning=@n102, Color_Day=@n102, Color_Evening=@n103, Color_Night=@n103) {materialFunction=&Color_Time_of_Day}.Result) {constB=0.1}
   Alpha=LinearInterpolate(Alpha=NamedRerouteUsage()) {constA=0.25}
  ) {constB=0}
  NamedRerouteDeclaration(None=LinearInterpolate(A=@n101, B=Add(A=@n101, B=@n104), Alpha=@n29)) {name=OutEmissive, nodeColor=<rgba 0.4766 1 0 1>}
  n105 = Constant() {r=0.1}
  n106 = MaterialFunctionCall(FlickerInitialSpeed=@n105, FlickerRandom=@n105) {materialFunction=&MF_Flicker}
  n107 = NamedRerouteUsage()
  n108 = Constant() {r=4}
  n109 = Multiply(
   A=LinearInterpolate(Alpha=@n106.Result) {constA=1, constB=2}
   B=Power(Base=OneMinus(None=@n107), Exp=Subtract(A=@n108, B=@n106.Result)) {constExponent=6}
  )
  n110 = NamedRerouteUsage()
  n111 = Add(A=@n109, B=Multiply(A=@n110, B=@n110))
  n112 = Multiply(A=Multiply(A=Multiply(A=@n111, B=@n109)), B=Power(Base=NamedRerouteUsage()))
  n113 = Multiply(A=Power(Base=@n110))
  NamedRerouteDeclaration(None=Saturate(None=Multiply(A=@n112, B=@n113))) {name=DetailFlames, nodeColor=<rgba 1 0 0.75 1>}
  n114 = LocalPosition() {localOrigin=InstancePreSkinning}
  NamedRerouteDeclaration(None=ShaderStageSwitch(PixelShader=VertexInterpolator(VS=@n114.XYZ).PS, VertexShader=@n114.XYZ)) {name=PreSkinnedPosition, nodeColor=<rgba 1 0.9141 0 1>}
  NamedRerouteDeclaration(None=@n1.BaseColor) {name=InBaseColor, nodeColor=<rgba 1 0.0469 0 1>}
  NamedRerouteDeclaration(None=@n1.SubsurfaceColor) {name=InSSSColor, nodeColor=<rgba 0 0.2734 1 1>}
  n115 = NamedRerouteUsage()
  n116 = NamedRerouteUsage()
  NamedRerouteDeclaration(None=LinearInterpolate(A=@n115, B=LinearInterpolate(B=@n115, Alpha=@n116) {constB=0}, Alpha=@n29)) {name=OutBaseColor, nodeColor=<rgba 0.5938 1 0 1>}
  n117 = NamedRerouteUsage()
  n118 = NamedRerouteUsage()
  NamedRerouteDeclaration(None=LinearInterpolate(A=@n117, B=LinearInterpolate(B=@n117, Alpha=@n118) {constB=0}, Alpha=@n29)) {name=OutSSSColor, nodeColor=<rgba 0 1 0.6719 1>}
  NamedRerouteUsage()
  NamedRerouteUsage()
 ]}
