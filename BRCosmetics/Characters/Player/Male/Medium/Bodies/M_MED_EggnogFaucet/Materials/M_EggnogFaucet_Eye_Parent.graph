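# Parent eye material for the EggnogFaucet character body, adding an emissive glow to the shared MF_eyes_2023 eye shading.
# Inverted roughness drives the strength of the EyeEmissive color (orange by default), so smoother areas glow brighter on top of the base emissive.
# A smoothstepped EffectMask scalar blends between plain and glowing attributes, on by default, before the result passes through MF_TechArtOverrides.
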Material /BRCosmetics/Characters/Player/Male/Medium/Bodies/M_MED_EggnogFaucet/Materials/M_EggnogFaucet_Eye_Parent
graph {
 properties {
  bUseMaterialAttributes=true
  bUsedWithMeshDeformer=true
  bUsedWithSkeletalMesh=true
  pixelDepthOffsetMode=PDOM_Legacy
  shadingModel=MSM_FromMaterialExpression
  usedShadingModels=Default Lit
 }
 nodes [
  n1 = MaterialFunctionCall() {materialFunction=/Game/Global/Materials/Character_Parent_2023/Eye_Parent_2023/MF_eyes_2023}
  n2 = GetMaterialAttributes(None=@n1.Result) {attributeGetTypes [#1, #2]}
  n3 = OneMinus(None=@n2.Roughness)
  n4 = Add(
   A=@n2.Emissive Color
   B=Multiply(A=LinearInterpolate(Alpha=@n3) {constA=0.1}, B=VectorParameter() {parameterName=EyeEmissive, defaultValue=<rgba 5 0.442 0 1>}.RGB)
  )
  n5 = ScalarParameter() {defaultValue=1, parameterName=EffectMask}
  MaterialFunctionCall(In=BlendMaterialAttributes(
    A=@n1.Result
    B=SetMaterialAttributes(MaterialAttributes=@n1.Result, Emissive Color=@n4) {attributeSetTypes [#1]}
    Alpha=SmoothStep(Value=@n5) {constMin=0.5}
   )) {materialFunction=&MF_TechArtOverrides}
 ]
}
